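# Material function for world position offset that turns a quad into a sprite facing a ray origin, the camera position by default.
# ImposterFrameTransform builds axes from the normalized center-to-origin direction; centered UVs, scaled by XY size, place vertices along them around the pivot.
# The offset output already subtracts world position.
# A second output flips Y and Z of optional normals and transforms them from view to tangent space; two camera-space Transform nodes are unconnected.
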MaterialFunction /Game/ImpostorBaker/MaterialFunctions/Sprite_PositionBased
graph {
 properties {
  bExposeToLibrary=true
  description=Use this function within the world position offset input to align plane meshes with the screen at any location.
  libraryCategoriesText [Misc, World Position Offset]
 }
 nodes [
  n1 = FunctionInput(Preview=TextureCoordinate()) {
   bUsePreviewValueAsDefault=true
   description=The quad should supply uvs that are mapped from the 0 to 1 uv range.
   inputName=0-1 UVs
   inputType=FunctionInput_Vector2
   previewValue=<0 0 0 0>
  }
  n2 = MaterialFunctionCall(Float2=Subtract(A=@n1) {constB=0.5}) {materialFunction=&BreakOutFloat2Components}
  n3 = FunctionInput(Preview=CameraPositionWS()) {bUsePreviewValueAsDefault=true, description=Ray Origin, inputName=Ray Origin, sortPriority=5, previewValue=<0 0 0 0>}
  n4 = FunctionInput(Preview=MaterialFunctionCall() {materialFunction=&ComponentPivotLocation}.Result) {
   bUsePreviewValueAsDefault=true
   description=The world position of the center of the sprite
   inputName=Center Location
   previewValue=<0 0 0 0>
  }
  n5 = MaterialFunctionCall!(InVec, X Vec, Y Vec, Z Vec=Normalize(VectorInput=Subtract(A=@n3, B=@n4))) {materialFunction=&/ImposterFrameTransform}
  n6 = FunctionInput() {bUsePreviewValueAsDefault=true, inputName=XY scale, inputType=FunctionInput_Vector2, previewValue=<50 100 0 1>}
  n7 = MaterialFunctionCall(Float2=@n6) {materialFunction=&BreakOutFloat2Components}
  n8 = Add(A=Multiply(A=Multiply(A=@n2.R, B=@n5.World X), B=@n7.R), B=Multiply(A=Multiply(A=@n2.G, B=@n5.World Y), B=@n7.G))
  n9 = WorldPosition()
  FunctionOutput(None=Add(A=Subtract(A=@n8, B=@n9.XYZ), B=@n4)) {
   description=Plug this into the world position offset input. Please note that world position is already subtracted from the output.
   outputName=World position offset
  }
  n10 = FunctionInput() {
   bUsePreviewValueAsDefault=true
   description=This is an optional input. The function will correctly transform your normals.
   inputName="Normals (Optional)"
   sortPriority=10
   previewValue=<0 0 1 1>
  }
  n11 = Constant3Vector() {constant=<rgba 1 -1 -1 0>}
  FunctionOutput(None=Transform(None=Multiply(A=@n10, B=@n11)) {transformSourceType=TRANSFORMSOURCE_View, transformType=TRANSFORM_Tangent}) {description=This will output transformed tangent space normals., outputName=Normals, sortPriority=1}
  Transform(None=Constant3Vector() {constant=<rgba 0 -1 0 0>}) {transformSourceType=TRANSFORMSOURCE_Camera}
  Transform(None=Constant3Vector() {constant=<rgba 1 0 0 0>}) {transformSourceType=TRANSFORMSOURCE_Camera}
 ]
}
