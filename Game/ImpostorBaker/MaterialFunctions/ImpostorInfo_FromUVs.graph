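# Material function for impostors that rebuilds world position and size from data stored in mesh UV channels.
# Position adds the world-space local origin, a local offset packed in UV2 and UV3.r, and the transformed Pivot Offset parameter scaled by UV3.g and object X scale.
# Default Size exposes the Default Mesh Size parameter, and Scaled Size multiplies it by object X scale and the UV3.g stored scale.
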MaterialFunction /Game/ImpostorBaker/MaterialFunctions/ImpostorInfo_FromUVs
graph {nodes [
  n1 = TransformPosition(None=Constant3Vector()) {transformType=TRANSFORMPOSSOURCE_World}
  n2 = VectorParameter() {parameterName=Pivot Offset}
  n3 = TextureCoordinate() {coordinateIndex=3}
  n4 = Multiply(
   A=Multiply(A=Transform(None=@n2.RGB), B=ComponentMask(None=@n3) {g=true})
   B=MaterialFunctionCall() {materialFunction=/Engine/Functions/Engine_MaterialFunctions02/WorldPositionOffset/ObjectScale}.Scale X
  )
  n5 = TextureCoordinate() {coordinateIndex=3}
  n6 = Multiply(A=AppendVector(A=TextureCoordinate() {coordinateIndex=2}, B=ComponentMask(None=@n5) {r=true}))
  FunctionOutput(None=Add(A=Add(A=@n1, B=@n4), B=Transform(None=@n6) {transformSourceType=TRANSFORMSOURCE_Local})) {outputName=Position}
  n7 = ScalarParameter() {defaultValue=350, parameterName=Default Mesh Size}
  n8 = MaterialFunctionCall() {materialFunction=&ObjectScale}
  n9 = ComponentMask(None=TextureCoordinate() {coordinateIndex=3}) {g=true}
  FunctionOutput(None=Multiply(A=FunctionOutput(None=@n7) {outputName=Default Size, sortPriority=1}, B=Multiply(A=@n8.Scale X, B=@n9))) {description=Multiplies Object Scale with the Stored UV Transform, outputName=Scaled Size, sortPriority=2}
  Transform(None=@n2.RGB) {transformSourceType=TRANSFORMSOURCE_Local}
 ]}
