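# Material function in the ImpostorBaker folder that encodes a unit vector as 2D hemi-octahedral coordinates, exposed to the material library.
# A single Custom HLSL node divides the input's XY by the vector's L1 norm, then outputs (x+y, x-y) to rotate and scale the diamond into a square.
# The result spans -1 to 1 with no remap to 0-1, and the sign of Z is discarded.
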MaterialFunction /Game/ImpostorBaker/MaterialFunctions/UnitVectorToHemiOctahedra
graph {
 properties {bExposeToLibrary=true}
 nodes [
  MaterialExpressionFunctionInput_0 = FunctionInput() {inputName=Unit Vector}
  n1 = Custom!(N=@MaterialExpressionFunctionInput_0) {
   code="\tN.xy /= dot( 1, abs(N) );\r\n\treturn float2( N.x + N.y, N.x - N.y );"
   outputType=CMOT_Float2
   inputs [{inputName=N, input {expression=&self:MaterialExpressionFunctionInput_0}}]
  }
  FunctionOutput(None=@n1) {outputName=Hemi Octahedron}
 ]
}
