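# Material function that renders a baked hemi-octahedral impostor using one nearest atlas frame, without blending, outputting base color, normal and WPO.
# View direction in tangent space snaps to a FramesXY grid cell, whose decoded direction orients the WPO-built billboard and selects the sub-UV.
# Normal alpha, sampled at MipForBleed, drives a parallax UV offset.
# Shadow-pass switches substitute a LightVector-based position for the camera and skip parallax.
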MaterialFunction /Game/ImpostorBaker/MaterialFunctions/Impostor_SingleFrameVersion
graph {nodes [
  n1 = MaterialFunctionCall() {materialFunction=&/ImpostorInfo}
  MaterialExpressionCameraPositionWS_6 = CameraPositionWS()
  MaterialExpressionMultiply_29 = Multiply(A=Normalize(VectorInput=VectorParameter() {parameterName=LightVector, defaultValue=<rgba -1 0 1 0>}.RGB)) {constB=500000}
  n2 = Custom!(Main=@MaterialExpressionCameraPositionWS_6, Shadow=@MaterialExpressionMultiply_29) {
   code="##if SHADOW_DEPTH_SHADER \r\n{\r\nreturn Shadow;\r\n}\r\n##endif\r\nreturn Main;"
   description=Shadow Pass Switch
   inputs [
    {inputName=Main, input {expression=&self:MaterialExpressionCameraPositionWS_6}}
    {inputName=Shadow, input {expression=&self:MaterialExpressionMultiply_29}}
   ]
  }
  n3 = MaterialFunctionCall() {materialFunction=&ImpostorInfo}
  n4 = Transform(None=Normalize(VectorInput=Subtract(A=@n2, B=@n3.Position))) {transformSourceType=TRANSFORMSOURCE_World, transformType=TRANSFORM_Tangent}
  n5 = MaterialFunctionCall(Float3=@n4) {materialFunction=&BreakOutFloat3Components}
  n6 = MaterialFunctionCall(X=@n5.R, Y=@n5.R, Z=Max(A=@n5.B) {constB=0.001}) {materialFunction=&MakeFloat3}
  n7 = ConstantBiasScale(None=MaterialFunctionCall!(Unit Vector=Normalize(VectorInput=@n6.Result)) {materialFunction=&/UnitVectorToHemiOctahedra}.Hemi Octahedron)
  n8 = Subtract(A=ScalarParameter() {defaultValue=16, parameterName=FramesXY})
  n9 = Add(A=Multiply(A=Clamp(None=@n7), B=@n8)) {constB=0.5}
  n10 = Divide(A=Floor(None=@n9), B=@n8)
  n11 = Normalize(VectorInput=MaterialFunctionCall!(Hemi Octahedron=ConstantBiasScale(None=@n10) {bias=-0.5, scale=2}) {materialFunction=&/HemiOctahedronToUnitVector}.Unit Vector) {desc=Z}
  n12 = MaterialFunctionCall(Float3=@n11) {materialFunction=&BreakOutFloat3Components}
  n13 = Normalize(VectorInput=MaterialFunctionCall(X=@n12.G, Y=Multiply(A=@n12.R) {constB=-1}, Z=Constant()) {materialFunction=&MakeFloat3}.Result) {desc=X}
  n14 = TextureCoordinate() {uTiling=10, vTiling=10}
  n15 = ScalarParameter() {defaultValue=16, parameterName=FramesXY}
  n16 = MaterialFunctionCall(Float2=ConstantBiasScale(None=Multiply(A=@n14, B=@n15)) {bias=-0.5, scale=2}) {materialFunction=&BreakOutFloat2Components}
  n17 = Multiply(A=@n13, B=@n16.R)
  n18 = Multiply(A=MaterialFunctionCall() {materialFunction=&ImpostorInfo}.Scaled Size) {constB=0.5}
  n19 = Normalize(VectorInput=CrossProduct(A=@n11, B=@n13)) {desc=Y}
  n20 = Multiply(A=@n19, B=@n16.G)
  n21 = Transform(None=Add(A=Multiply(A=@n17, B=@n18), B=Multiply(A=@n20, B=@n18)))
  Add(A=@n1.Position, B=@n21)
  CrossProduct(A=@n11, B=Constant3Vector() {constant=<rgba 0 0 1 0>})
  n22 = VertexInterpolator(VS=@n9)
  n23 = ScalarParameter() {defaultValue=16, parameterName=FramesXY}
  n24 = Multiply(A=TextureCoordinate() {uTiling=10, vTiling=10}, B=@n23)
  n25 = Add(A=Divide(A=Floor(None=@n22.PS), B=ScalarParameter() {defaultValue=16, parameterName=FramesXY}), B=Divide(A=Saturate(None=@n24), B=@n23))
  MaterialExpressionReroute_8 = Reroute(None=@n25)
  n26 = ScalarParameter() {defaultValue=1, parameterName=MipForBleed}
  n27 = VertexInterpolator(VS=MaterialFunctionCall!(VectorToTransform=@n4, BasisX=@n13, BasisY=@n19, BasisZ=@n11) {materialFunction=/Engine/Functions/Engine_MaterialFunctions02/Math/InverseTransformMatrix}.Transformed Vector 3)
  n28 = Multiply(
   A=Subtract(A=TextureSampleParameter2D!(UVs=@n25, Level=@n26, Apply View MipBias) {constMipValue=0, mipValueMode=TMVM_MipLevel, parameterName=Normal, texture=&Black}.A) {constA=0.5, constB=0.5}
   B=Divide(A=ComponentMask(None=@n27.PS) {g=true, r=true}, B=ScalarParameter() {defaultValue=16, parameterName=FramesXY}) {constB=16}
  )
  MaterialExpressionAdd_0 = Add(A=@MaterialExpressionReroute_8, B=@n28)
  n29 = Custom(Main=@MaterialExpressionAdd_0, Shadow=@MaterialExpressionReroute_8) {
   code="##if SHADOW_DEPTH_SHADER \r\n{\r\nreturn Shadow;\r\n}\r\n##endif\r\nreturn Main;"
   description=Shadow Pass Switch
   outputType=CMOT_Float2
   inputs [
    {inputName=Main, input {expression=&self:MaterialExpressionAdd_0}}
    {inputName=Shadow, input {expression=&self:MaterialExpressionReroute_8}}
   ]
  }
  n30 = TextureSampleParameter2D!(UVs=@n29, Apply View MipBias) {parameterName=BaseColor, texture=&Black}
  FunctionOutput(None=@n30.RGB) {outputName=BaseColor RGB}
  FunctionOutput(None=@n30.A) {outputName=BaseColor A, sortPriority=1}
  n31 = TextureSampleParameter2D(UVs=@n29) {parameterName=Normal, texture=&Black}
  FunctionOutput(None=@n31.RGB) {outputName=Normal RGB, sortPriority=2}
  FunctionOutput(None=@n31.A) {outputName=Normal A, sortPriority=3}
  n32 = Multiply(A=WorldPosition() {worldPositionShaderOffset=WPT_ExcludeAllShaderOffsets}.XYZ) {constB=-1}
  FunctionOutput(None=Add(A=Add(A=@n21, B=@n32), B=MaterialFunctionCall() {materialFunction=&ImpostorInfo}.Position) {constB=0}) {outputName=WPO, sortPriority=4}
  Multiply(A=@n28, B=ScalarParameter() {defaultValue=1, parameterName=EnableParallax})
  Normalize(VectorInput=@n13) {desc=X}
  MaterialExpressionCameraPositionWS_4 = CameraPositionWS()
  n33 = VectorParameter() {parameterName=LightVector, defaultValue=<rgba -1 0 1 0>}
  MaterialExpressionMultiply_9 = Multiply(A=Multiply(A=Normalize(VectorInput=@n33.RGB)) {constB=-1}) {constB=500000}
  n34 = Custom(Main=@MaterialExpressionCameraPositionWS_4, Shadow=@MaterialExpressionMultiply_9) {
   code="##if SHADOW_DEPTH_SHADER \r\n{\r\nreturn Shadow;\r\n}\r\n##endif\r\nreturn Main;"
   description=Shadow Pass Switch
   inputs [
    {inputName=Main, input {expression=&self:MaterialExpressionCameraPositionWS_4}}
    {inputName=Shadow, input {expression=&self:MaterialExpressionMultiply_9}}
   ]
  }
  Transform(None=Normalize(VectorInput=Subtract(A=@n34, B=@n1.Position))) {transformSourceType=TRANSFORMSOURCE_World, transformType=TRANSFORM_Tangent}
 ]}
