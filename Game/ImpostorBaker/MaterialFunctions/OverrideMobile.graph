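# OverrideMobile is a material function in the ImpostorBaker content that lets a material supply a separate value for the mobile shading path.
# It takes a Default input and a Mobile Override input and routes them through a ShadingPathSwitch, noted as migrated from a FeatureLevelSwitch, to a single output.
# Both inputs use their preview values as defaults, so either can be left unconnected.
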MaterialFunction /Game/ImpostorBaker/MaterialFunctions/OverrideMobile
graph {nodes [FunctionOutput(None=ShadingPathSwitch(
    Default=FunctionInput() {bUsePreviewValueAsDefault=true, inputName=Default}
    Mobile=FunctionInput() {bUsePreviewValueAsDefault=true, inputName=Mobile Override, sortPriority=1}
   ) {desc=Migrated from FeatureLevelSwitch})]}
