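# Material function in ImpostorBaker that computes three triangle interpolation weights and a mask from a UV input, which previews with texture coordinates.
# It takes the fractional UV within each cell, then builds a float3 from Min nodes on the inverted and plain components around the absolute U−V difference from a dot product.
# A second output, Mask, saturates the ceiling of a component subtraction to give a hard triangle selector.
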MaterialFunction /Game/ImpostorBaker/MaterialFunctions/TriangleInterpolator
graph {nodes [
  n1 = Frac(None=FunctionInput(Preview=TextureCoordinate()) {bUsePreviewValueAsDefault=true, inputName=UV, inputType=FunctionInput_Vector2})
  n2 = MaterialFunctionCall(Float2=OneMinus(None=@n1)) {materialFunction=&BreakOutFloat2Components}
  n3 = DotProduct(A=@n1, B=Constant2Vector() {g=-1, r=1})
  n4 = MaterialFunctionCall(Float2=@n1) {materialFunction=&BreakOutFloat2Components}
  FunctionOutput(None=MaterialFunctionCall(X=Min(A=@n2.R, B=@n2.R) {desc=1}, Y=Abs(None=@n3) {desc=3}, Z=Min(A=@n4.R, B=@n4.R) {desc=2}) {materialFunction=&MakeFloat3}.Result)
  n5 = Subtract(A=@n4.R, B=@n4.R)
  FunctionOutput(None=Saturate(None=Ceil(None=@n5))) {outputName=Mask, sortPriority=1}
 ]}
