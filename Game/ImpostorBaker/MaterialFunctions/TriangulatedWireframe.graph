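# Material function that generates a triangulated wireframe mask from a UV input and a Line Width input, kept with the Impostor Baker.
# The fractional part of the UV repeats the pattern per unit cell, with a centre-distance term for border lines and a component-difference term for the diagonal, its width scaled by 1.41.
# Ceil and Floor harden the two terms, and Max combines them into one output.
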MaterialFunction /Game/ImpostorBaker/MaterialFunctions/TriangulatedWireframe
graph {nodes [
  n1 = FunctionInput(Preview=TextureCoordinate()) {bUsePreviewValueAsDefault=true, inputName=UV, inputType=FunctionInput_Vector2}
  n2 = Subtract(A=Frac(None=@n1)) {constB=0.5}
  n3 = MaterialFunctionCall(Float2=Multiply(A=Abs(None=@n2)) {constB=2}) {materialFunction=&BreakOutFloat2Components}
  n4 = FunctionInput() {bUsePreviewValueAsDefault=true, inputName=Line Width, inputType=FunctionInput_Scalar, sortPriority=1, previewValue=<0.02 0 0 1>}
  n5 = Subtract(A=Max(A=@n3.R, B=@n3.R), B=OneMinus(None=@n4))
  n6 = MaterialFunctionCall(Float2=Frac(None=@n1)) {materialFunction=&BreakOutFloat2Components}
  n7 = Abs(None=Subtract(A=@n6.R, B=@n6.R))
  n8 = OneMinus(None=Subtract(A=Divide(A=@n7) {constB=1}, B=Divide(A=@n4) {constB=1.41}))
  FunctionOutput(None=Max(A=Ceil(None=@n5), B=Floor(None=@n8)))
 ]}
